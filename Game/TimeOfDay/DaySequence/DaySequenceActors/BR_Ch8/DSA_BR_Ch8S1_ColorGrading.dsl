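; Day sequence actor for Battle Royale Chapter 8 Season 1 color grading, with a switch for showing its baked clouds.
; Its construction script calls the parent, then runs BakedCloudsVisibility to apply the ShowBakedClouds flag.
; That function shows or hides BakedClouds by setting its world location to the origin or far below the world, rather than toggling visibility.
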(fn BakedCloudsVisibility
  (Transformation|SetWorldLocation (Variables|Default|GetBakedClouds) (select (Variables|Default|GetShowBakedClouds) "0, 0, 0" "0.000000,0.000000,-10000000.000000") false false))

(fn ConstructionScript
  (|CallParentFunction)
  (CallFunction|BakedCloudsVisibility))
