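# Prototype sheen layer material function that modifies the base color and emissive of incoming material attributes.
# A sky-light and camera vector dotted with normals, distorted by a repeating local-position pattern, forms a contrast-sharpened highlight with face and layer masking.
# The highlight scales base color tinted between shadow and light colors by sun facing.
# Emissive is scaled by per-time-of-day parameters and rises from half to full with pixel depth.
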MaterialFunction /Game/TimeOfDay/DaySequence/DaySequenceActors/BR_Ch8/MF_4Layer_Sheen_Proto
graph {nodes [
  n1 = NamedRerouteUsage()
  n2 = NamedRerouteUsage()
  n3 = Transform(None=@n2)
  n4 = NamedRerouteUsage()
  n5 = Multiply(A=DotProduct(A=SkyAtmosphereLightDirection(), B=StaticSwitch(True=@n3, False=@n2, Value=@n4))) {constB=20}
  n6 = LinearInterpolate(
   A=Multiply(A=@n1, B=VectorParameter() {group=Sheen, parameterName=Sheen Diffuse Shadow, defaultValue=<rgba 0.5781 0.5781 0.5781 1>}.RGB)
   B=Multiply(A=@n1, B=VectorParameter() {group=Sheen, parameterName=Sheen Diffuse Light, defaultValue=<rgba 1 1 1 1>}.RGB)
   Alpha=Saturate(None=Add(A=@n5) {constB=0.5})
  ) {constA=0.5}
  n7 = Add(
   A=Multiply(
    A=SkyAtmosphereLightDirection()
    B=ScalarParameter() {defaultValue=1, group=Sheen, parameterName=Sheen Specular View Dependency, sliderMax=1, sortPriority=0}
   ) {constB=0.7}
   B=CameraVectorWS()
  )
  n8 = NamedRerouteUsage()
  n9 = LinearInterpolate(
   A=StaticSwitch(True=Transform(None=@n8), False=@n8, Value=NamedRerouteUsage())
   B=Normalize(VectorInput=VertexNormalWS())
   Alpha=ScalarParameter() {defaultValue=0.2, group=Sheen, parameterName=Sheen Vertex Normals, sliderMax=1, sortPriority=0}
  )
  n10 = NamedRerouteDeclaration(None=@n9) {name=Normals, nodeColor=<rgba 0 0.2031 1 1>}
  n11 = Frac(None=Divide(A=LocalPosition().XYZ, B=ScalarParameter() {defaultValue=300, group=Sheen, parameterName=Sheen Specular Distortion Scale, sortPriority=0}) {constB=300})
  n12 = Abs(None=Multiply(A=Add(A=@n11) {constB=-0.5}) {constB=2})
  n13 = OneMinus(None=Multiply(A=@n12, B=@n12) {constB=0.1})
  n14 = Multiply(A=@n13, B=@n13)
  n15 = Multiply(
   A=Add(A=OneMinus(None=@n14)) {constB=-0.5}
   B=ScalarParameter() {defaultValue=0.052, group=Sheen, parameterName=Sheen Distortion Intensity, sortPriority=0}
  ) {constB=0.04}
  n16 = ScalarParameter() {group=Sheen, parameterName=Sheen Mask Vertical Faces, sortPriority=0}
  n17 = Multiply(
   A=Abs(None=NamedRerouteUsage())
   B=AppendVector(A=AppendVector(A=@n16, B=@n16), B=ScalarParameter() {group=Sheen, parameterName=Sheen Mask Horizontal Faces, sortPriority=0})
  )
  n18 = Constant() {r=1}
  n19 = Subtract(A=DotProduct(A=Normalize(VectorInput=@n7), B=Add(A=@n10, B=@n15)), B=Saturate(None=DotProduct(A=@n17, B=@n18)))
  n20 = Add(A=ScalarParameter() {defaultValue=0.8, group=Sheen, parameterName=Sheen Size, sortPriority=0}) {constB=-1}
  n21 = Saturate(None=MaterialFunctionCall(In=Add(A=@n19, B=@n20) {constB=0}, Contrast=ScalarParameter() {defaultValue=10, group=Sheen, parameterName=Sheen Contrast, sortPriority=0}) {materialFunction=&CheapContrast}.Result)
  n22 = Saturate(None=LinearInterpolate(
    A=ScalarParameter() {group=Sheen, parameterName=Sheen Mask Low, sortPriority=0}
    B=ScalarParameter() {defaultValue=1, group=Sheen, parameterName=Sheen Mask High, sortPriority=0}
    Alpha=NamedRerouteUsage()
   ))
  n23 = Add(A=Multiply(A=Multiply(A=@n21, B=@n22), B=ScalarParameter() {group=Sheen, parameterName=Sheen Intensity, sortPriority=0}))
  n24 = Multiply(
   A=NamedRerouteUsage()
   B=MaterialFunctionCall(
    Value_Morning=ScalarParameter() {defaultValue=1, group=Sheen, parameterName=Sheen Emissive Morning, sliderMax=1, sortPriority=0}
    Value_Day=ScalarParameter() {defaultValue=1, group=Sheen, parameterName=Sheen Emissive Day, sliderMax=1, sortPriority=0}
    Value_Evening=ScalarParameter() {defaultValue=1, group=Sheen, parameterName=Sheen Emissive Evening, sliderMax=1, sortPriority=0}
    Value_Night=ScalarParameter() {defaultValue=1, group=Sheen, parameterName=Sheen Emissive Night, sliderMax=1, sortPriority=0}
   ) {materialFunction=&Scalar_Time_of_Day}.Result
  )
  n25 = PixelDepth()
  n26 = LinearInterpolate(Alpha=Saturate(None=Divide(A=@n25.R) {constB=800})) {constA=0.5}
  FunctionOutput(None=SetMaterialAttributes!(MaterialAttributes=NamedRerouteUsage(), BaseColor=Multiply(A=@n6, B=@n23), EmissiveColor=Multiply(A=@n24, B=@n26)) {attributeSetTypes [#1, #2]})
  n27 = FunctionInput() {bUsePreviewValueAsDefault=true, inputName=Material, inputType=FunctionInput_MaterialAttributes, previewValue=<1 1 1 1>}
  n28 = GetMaterialAttributes(None=@n27) {attributeGetTypes [
    #1
    #3
    #4
    #2
   ]}
  NamedRerouteDeclaration(None=@n28.MaterialAttributes) {name=Material Input, nodeColor=<rgba 1 0.4922 0 1>}
  n29 = FunctionInput() {bUsePreviewValueAsDefault=true, inputName=Layer Mask, inputType=FunctionInput_Vector4, sortPriority=1, previewValue=<1 1 1 1>}
  NamedRerouteDeclaration(None=@n29) {name=Layer Mask, nodeColor=<rgba 1 0 0.3471 1>}
  NamedRerouteDeclaration(None=@n28.BaseColor) {name=BaseColor Input, nodeColor=<rgba 1 0 0.2109 1>}
  NamedRerouteDeclaration(None=@n28.Normal) {name=Normal Input, nodeColor=<rgba 0 0.5312 1 1>}
  NamedRerouteDeclaration(None=FunctionInput() {bUsePreviewValueAsDefault=true, inputName=TangentNormals, inputType=FunctionInput_StaticBool}) {name=Tangent Normals, nodeColor=<rgba 1 0.3984 0 1>}
  NamedRerouteDeclaration(None=@n28.Specular) {name=Specular Input, nodeColor=<rgba 0 0.0859 1 1>}
  NamedRerouteDeclaration(None=@n28.EmissiveColor) {name=Emissive Input, nodeColor=<rgba 1 0 0.1875 1>}
  n30 = FunctionInput() {bUsePreviewValueAsDefault=true, inputName=Layer Texture Mask, inputType=FunctionInput_Scalar, sortPriority=2, previewValue=<1 1 1 1>}
  NamedRerouteDeclaration(None=@n30) {name=Layer Texture Mask, nodeColor=<rgba 0 0.5312 1 1>}
  Reroute(None=@n9)
 ]}
