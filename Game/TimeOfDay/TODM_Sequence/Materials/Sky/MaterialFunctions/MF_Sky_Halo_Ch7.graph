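# Sky material function that adds a tinted sun bloom and halo, plus a two-color additive gradient, to the Sky Atmosphere input, limited by a Horizon Mask input.
# Dot and cross products of the atmosphere light direction and a mirrored camera vector shape the glow, blended by Halo Squash.
# A backside mask from the remapped dot product fades the glow and drives the gradient.
# A second output supplies a fixed-shape Halo Mask.
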MaterialFunction /Game/TimeOfDay/TODM_Sequence/Materials/Sky/MaterialFunctions/MF_Sky_Halo_Ch7
graph {nodes [
  n1 = Multiply(A=DotProduct(
    A=SkyAtmosphereLightDirection()
    B=MaterialFunctionCall!(Multiplier, Offset) {materialFunction=/Game/TimeOfDay/TODM_Sequence/Materials/Sky/MF_MirroredCameraVector}.Result
   )) {constB=-1}
  n2 = Saturate(None=@n1)
  n3 = CrossProduct(
   A=SkyAtmosphereLightDirection()
   B=MaterialFunctionCall() {materialFunction=&MF_MirroredCameraVector}.Result
  )
  n4 = OneMinus(None=Distance(A=ComponentMask(None=@n3) {g=true, r=true}, B=Constant()))
  n5 = LinearInterpolate(A=@n2, B=@n4, Alpha=ScalarParameter() {parameterName=Halo Squash}) {constAlpha=0}
  n6 = Multiply(
   A=Power(Base=@n5, Exp=ScalarParameter() {defaultValue=40, parameterName=SunBloomExp})
   B=VectorParameter() {parameterName=SunBloomColor, defaultValue=<rgba 0 0 0 1>}.RGB
  )
  n7 = Multiply(
   A=Power(Base=@n5, Exp=ScalarParameter() {defaultValue=4, parameterName=Halo Exp})
   B=VectorParameter() {parameterName=Halo Color, defaultValue=<rgba 0 0 0 1>}.RGB
  )
  n8 = OneMinus(None=NamedRerouteUsage())
  n9 = ScalarParameter() {defaultValue=1, parameterName=Additive Color Range}
  n10 = Add(
   A=Multiply(A=Add(A=@n6, B=@n7), B=NamedRerouteUsage())
   B=LinearInterpolate(
    A=VectorParameter() {parameterName=Color Additive, defaultValue=<rgba 0 0 0 1>}.RGB
    B=VectorParameter() {parameterName=Color Additive 2, defaultValue=<rgba 0 0 0 1>}.RGB
    Alpha=Power(Base=@n8, Exp=@n9)
   )
  )
  n11 = FunctionInput() {bUsePreviewValueAsDefault=true, inputName=Horizon Mask, inputType=FunctionInput_Scalar, sortPriority=1, previewValue=<1 1 1 1>}
  FunctionOutput(None=Add(A=Multiply(A=@n10, B=@n11), B=FunctionInput() {inputName=Sky Atmosphere}))
  n12 = LinearInterpolate(A=@n2, B=@n4) {constAlpha=0.7}
  n13 = Multiply(A=@n12, B=@n12)
  FunctionOutput(None=Multiply(A=Multiply(A=@n13, B=@n13), B=NamedRerouteUsage())) {outputName=Halo Mask, sortPriority=1}
  n14 = Add(A=Multiply(A=@n1) {constB=0.5}) {constB=0.5}
  NamedRerouteDeclaration(None=@n14) {name=Backside Mask, nodeColor=<rgba 1 0 0.0234 1>}
  Reroute(None=@n14)
 ]}
