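# Material function for the time-of-day sky that adds twinkling stars onto an incoming color.
# It samples the T_Ch8_Stars texture at two tilings, squaring red for star intensity and using green as per-star phase for time-driven triangle waves at differing rates.
# Each layer is multiplied by the other's blue channel, then both are summed, tinted by the Stars vector parameter, boosted and masked by the Mask input before the add.
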MaterialFunction /Game/TimeOfDay/TODM_Sequence/Materials/Sky/MaterialFunctions/MF_Sky_Stars_Ch8
graph {nodes [
  n1 = TextureCoordinate()
  n2 = TextureSample(UVs=Multiply(A=@n1) {constB=20}) {samplerType=SAMPLERTYPE_LinearColor, texture=/Game/TimeOfDay/TODM_Sequence/Materials/Sky/Textures/T_Ch8_Stars}
  n3 = AppendVector(A=@n2.G, B=@n2.G)
  n4 = TextureSample(UVs=Multiply(A=@n1) {constB=7}) {samplerType=SAMPLERTYPE_LinearColor, texture=&T_Ch8_Stars}
  n5 = AppendVector(A=@n4.G, B=@n4.G)
  n6 = Time()
  n7 = Constant4Vector() {constant=<rgba 0.34 0.686 0.46 0.712>}
  n8 = Frac(None=Add(A=AppendVector(A=@n3, B=@n5), B=Multiply(A=@n6, B=@n7.RGBA) {constB=0.5}))
  MaterialExpressionMultiply_25 = Multiply(A=Abs(None=Add(A=@n8) {constB=-0.5})) {constB=2}
  n9 = Convert(None=@MaterialExpressionMultiply_25) {
   nodeName=Break Float4
   convertInputs [{
     type=Vector4
     defaultValue=<rgba 0 0 0 1>
     expressionInput {
      expression=&self:MaterialExpressionMultiply_25
      inputName=None
      mask=0
      maskA=0
      maskB=0
      maskG=0
      maskR=0
      outputIndex=0
     }
    }]
   convertOutputs [{type=Scalar}, {type=Scalar}, {type=Scalar}, {type=Scalar}]
  }
  n10 = Add(A=@n9, B=@n9)
  n11 = Multiply(A=Multiply(A=Multiply(A=@n2.R, B=@n2.R), B=Saturate(None=@n10)) {constB=0.6}) {constB=0.2}
  n12 = Multiply(A=Multiply(A=@n4.R, B=@n4.R), B=Saturate(None=Add(A=@n9, B=@n9)))
  n13 = VectorParameter() {group=Stars, parameterName=Stars, defaultValue=<rgba 0 0 0 1>}
  n14 = Multiply(
   A=Add(A=Multiply(A=@n11, B=@n4.B), B=Multiply(A=@n2.B, B=@n12))
   B=Multiply(A=ComponentMask(None=@n13.RGB) {b=true, g=true, r=true}) {constB=25}
  )
  n15 = FunctionInput() {bUsePreviewValueAsDefault=true, inputName=Mask, inputType=FunctionInput_Scalar, sortPriority=1, previewValue=<1 1 1 1>}
  FunctionOutput(None=Add(A=Multiply(A=@n14, B=@n15), B=FunctionInput() {bUsePreviewValueAsDefault=true, inputName=Color}))
 ]}
