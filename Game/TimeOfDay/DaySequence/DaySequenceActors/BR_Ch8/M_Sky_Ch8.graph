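# Unlit sky material for the BR Chapter 8 day sequence, layering halo, high-altitude clouds and stars over Sky Atmosphere luminance sampled along a mirrored camera vector.
# A horizon mask from powered vertex-color red shapes the layers, with stars also masked by clouds.
# Storm-circle darkening, post-process fog color and a negative-desaturation saturation boost finish the main view.
# Reflection capture passes switch to separate reflection color and world position offset from MF_Sky_Reflection.
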Material /Game/TimeOfDay/DaySequence/DaySequenceActors/BR_Ch8/M_Sky_Ch8
graph {
 properties {bAutomaticallySetUsageInEditor=false, bIsSky=true, shadingModel=MSM_Unlit, usedShadingModels=Unlit}
 nodes [
  n1 = VertexColor()
  n2 = Multiply(A=@n1.R, B=@n1.R)
  n3 = Multiply(A=@n2, B=@n2)
  n4 = Multiply(A=@n3, B=@n3)
  NamedRerouteDeclaration(None=OneMinus(None=Multiply(A=@n4) {constB=0.9})) {name=Horizon Mask, nodeColor=<rgba 0 1 0.7422 1>}
  n5 = Constant() {r=-1}
  n6 = NamedRerouteUsage()
  n7 = MaterialFunctionCall!(
   Sky Atmosphere=SkyAtmosphereViewLuminance!(WorldDirection=MaterialFunctionCall!(Multiplier=@n5, Offset) {materialFunction=/Game/TimeOfDay/TODM_Sequence/Materials/Sky/MF_MirroredCameraVector}.Result)
   Horizon Mask=@n6
  ) {materialFunction=/Game/TimeOfDay/TODM_Sequence/Materials/Sky/MaterialFunctions/MF_Sky_Halo_Ch7}
  n8 = MaterialFunctionCall!(Color=@n7.Result, Halo Mask=@n7.Result, Horizon Mask=@n6) {materialFunction=/Game/TimeOfDay/TODM_Sequence/Materials/Sky/MaterialFunctions/MF_Sky_HighAltitudeClouds_Ch8}
  n9 = Multiply(A=OneMinus(None=@n8.Mask), B=@n6)
  n10 = MaterialFunctionCall!(
   Color=NamedRerouteDeclaration(None=MaterialFunctionCall!(Color=@n8.Color, Mask=@n9) {materialFunction=/Game/TimeOfDay/TODM_Sequence/Materials/Sky/MaterialFunctions/MF_Sky_Stars_Ch8}.Result) {name=Pre Skylight Captures, nodeColor=<rgba 0 1 0.5781 1>}
   Storm Color=VectorParameter() {parameterName=Storm Darkening, defaultValue=<rgba 0.1332 0.0989 0.4219 1>}.RGB
   Height
  ) {materialFunction=/Game/TimeOfDay/CloudsAndStorm/Materials/MaterialFunctions/MF_Storm_DarkeningCircle}
  n11 = CollectionParameter() {collection=&FortniteMaterialParameters, parameterName=FogPostProcessColor}
  n12 = ComponentMask(None=@n11) {b=true, g=true, r=true}
  n13 = MaterialFunctionCall!(Color=NamedRerouteUsage()) {materialFunction=/Game/TimeOfDay/TODM_Sequence/Materials/Sky/MaterialFunctions/MF_Sky_Reflection}
  ReflectionCapturePassSwitch!(Default=Desaturation(None=Add(A=@n10.Result, B=@n12), Fraction=Constant() {r=-0.4}), Reflection=@n13.Reflection Color)
  ReflectionCapturePassSwitch(Default=Constant(), Reflection=@n13.Reflection WPO)
 ]
}
