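# Light function material in the BR_Ch8 day sequence folder that defines a night-only mask by time of day.
# A single Scalar_Time_of_Day material function call receives zero for morning, day and evening, and one for night.
# The graph uses only constants, with no textures or parameters of its own.
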Material /Game/TimeOfDay/DaySequence/DaySequenceActors/BR_Ch8/M_TODLight
graph {
 properties {bAutomaticallySetUsageInEditor=false, materialDomain=MD_LightFunction, usedShadingModels=Default Lit}
 nodes [
  n1 = Constant()
  MaterialFunctionCall(Value_Morning=@n1, Value_Day=@n1, Value_Evening=@n1, Value_Night=Constant() {r=1}) {materialFunction=&Scalar_Time_of_Day}
 ]
}
