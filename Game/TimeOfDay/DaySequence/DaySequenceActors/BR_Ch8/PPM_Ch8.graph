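# Post-process material for BR Chapter 8 that adds stylized outlines, light-facing edge highlights and gamma/saturation grading, applied after DOF.
# Outlines come from four-tap scene depth and normal comparisons, with custom depth, shading-model water/vegetation masks and distance fades.
# Highlights offset depth samples along the view-space sky light direction, with widths varied by a world-position pattern.
# MPC_Ch8_PP drives the look; a shading-path switch and color-difference posterized lines cover mobile.
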Material /Game/TimeOfDay/DaySequence/DaySequenceActors/BR_Ch8/PPM_Ch8
graph {
 properties {
  bAutomaticallySetUsageInEditor=false
  blendableLocation=BL_SceneColorAfterDOF
  materialDomain=MD_PostProcess
  previewMesh=/Hera/Sets/SeaMist/Meshes/SM_SeaMist_Lighthouse_FresnelLens_A
  usedShadingModels=Default Lit
 }
 nodes [
  n1 = EyeAdaptationInverse(LightValueInput=NamedRerouteUsage(), AlphaInput=Constant() {r=0.5})
  n2 = NamedRerouteUsage()
  n3 = Multiply(A=CollectionParameter() {collection=&/MPC_Ch8_PP, parameterName=OutlineSaturation}) {constB=-1}
  n4 = LinearInterpolate(
   A=Multiply(
    A=Desaturation(None=@n2, Fraction=@n3)
    B=CollectionParameter() {collection=&MPC_Ch8_PP, parameterName=OutlineMultiply}
   ) {constB=0.76}
   B=NamedRerouteUsage()
   Alpha=NamedRerouteUsage()
  )
  n5 = CollectionParameter() {collection=&MPC_Ch8_PP, parameterName=HighlightSaturation}
  n6 = Min(A=DotProduct(A=Constant3Vector() {constant=<rgba 0.21 0.71 0.072 1>}, B=NamedRerouteUsage()))
  n7 = Multiply(
   A=Desaturation(None=NamedRerouteUsage(), Fraction=Multiply(A=@n5) {constB=-1})
   B=Multiply(
    A=LinearInterpolate(Alpha=@n6) {constA=1.5, constB=1.1}
    B=CollectionParameter() {collection=&MPC_Ch8_PP, parameterName=HighlightMultiply}
   )
  ) {constB=1.1}
  n8 = NamedRerouteUsage()
  n9 = Multiply(A=Multiply(A=NamedRerouteUsage(), B=OneMinus(None=@n8)), B=NamedRerouteUsage())
  n10 = NamedRerouteUsage()
  n11 = CollectionParameter() {collection=&MPC_Ch8_PP, parameterName=GammaCharacterWeight}
  n12 = LinearInterpolate(A=NamedRerouteUsage(), B=Max(A=LinearInterpolate(A=@n4, B=@n7, Alpha=@n9)) {constB=0}, Alpha=Saturate(None=Add(A=@n10, B=@n11) {constB=0.5}))
  n13 = NamedRerouteUsage()
  n14 = Sine(None=Multiply(A=Time()) {constB=0.35})
  LinearInterpolate(
   A=NamedRerouteUsage()
   B=ShadingPathSwitch(Default=LinearInterpolate(A=@n1.EyeAdaptationInverse, B=@n12, Alpha=@n13), Mobile=NamedRerouteUsage())
   Alpha=Saturate(None=Ceil(None=@n14))
  )
  n15 = NamedRerouteUsage()
  n16 = NamedRerouteUsage()
  n17 = Max(A=Saturate(None=Min(A=@n15, B=@n16)), B=NamedRerouteUsage())
  n18 = Constant()
  n19 = MaterialFunctionCall!(ID=Constant() {r=6}) {
   desc=adding because items want outline but need custom depth
   materialFunction=/Game/Packages/Fortress_SharedMaterials/Stencil_ID_Definitions/StencilMaskCompare
  }
  n20 = OneMinus(None=@n19.Result)
  Max(A=Max(A=@n17, B=Multiply(A=@n18, B=@n20)), B=NamedRerouteUsage())
  n21 = ComponentMask(None=SceneTexture(UVs=NamedRerouteUsage()) {sceneTextureId=PPI_ShadingModelID}.Color) {r=true}
  n22 = ComponentMask(None=SceneTexture(UVs=NamedRerouteUsage()) {sceneTextureId=PPI_ShadingModelID}.Color) {r=true}
  n23 = ComponentMask(None=SceneTexture(UVs=NamedRerouteUsage()) {sceneTextureId=PPI_ShadingModelID}.Color) {r=true}
  n24 = ComponentMask(None=SceneTexture(UVs=NamedRerouteUsage()) {sceneTextureId=PPI_ShadingModelID}.Color) {r=true}
  Min(A=Min(A=@n21, B=@n22), B=Min(A=@n23, B=@n24))
  n25 = NamedRerouteUsage()
  n26 = NamedRerouteUsage()
  n27 = Ceil(None=SceneTexture(UVs=NamedRerouteUsage()) {sceneTextureId=PPI_Specular}.Color)
  n28 = LinearInterpolate(A=@n25, B=@n26, Alpha=@n27)
  n29 = Ceil(None=SceneTexture(UVs=NamedRerouteUsage()) {sceneTextureId=PPI_Specular}.Color)
  n30 = LinearInterpolate(A=@n26, B=@n25, Alpha=@n29)
  n31 = NamedRerouteUsage()
  n32 = NamedRerouteUsage()
  n33 = Ceil(None=SceneTexture(UVs=NamedRerouteUsage()) {sceneTextureId=PPI_Specular}.Color)
  n34 = LinearInterpolate(A=@n31, B=@n32, Alpha=@n33)
  n35 = Ceil(None=SceneTexture(UVs=NamedRerouteUsage()) {sceneTextureId=PPI_Specular}.Color)
  n36 = LinearInterpolate(A=@n32, B=@n31, Alpha=@n35)
  n37 = Add(A=Add(A=Subtract(A=@n28, B=@n25), B=Subtract(A=@n30, B=@n26)), B=Add(A=Subtract(A=@n34, B=@n31), B=Subtract(A=@n36, B=@n32)))
  n38 = Multiply(A=NamedRerouteUsage())
  Multiply(A=Saturate(None=Step(Y=@n37, X=@n38)), B=Max(A=Max(A=@n27, B=@n29), B=Max(A=@n33, B=@n35)))
  n39 = Multiply(A=ComponentMask(None=ViewSize()) {g=true}, B=ViewProperty() {property=MEVP_ResolutionFraction}.Property)
  n40 = ScalarParameter() {defaultValue=900, parameterName=VerticalResolutionThreshold}
  NamedRerouteDeclaration(None=Saturate(None=Subtract(A=@n39, B=@n40))) {name=Is High Res, nodeColor=<rgba 1 0 0.1406 1>}
  n41 = LinearInterpolate(
   A=ScalarParameter() {defaultValue=80, group=Outlines Depth, parameterName=DepthThreshold, sliderMax=30, sortPriority=1}
   B=ScalarParameter() {defaultValue=800, group=Outlines Depth, parameterName=DepthThreshold Vegetation, sliderMax=30, sortPriority=1}
   Alpha=NamedRerouteUsage()
  )
  n42 = ComponentMask(None=SceneTexelSize()) {g=true}
  n43 = NamedRerouteUsage()
  n44 = Multiply(A=Add(A=Multiply(A=@n41, B=@n42), B=Multiply(A=@n43) {constB=0.1}), B=NamedRerouteUsage())
  n45 = ScreenPosition()
  n46 = SkyAtmosphereLightDirection()
  n47 = Constant3Vector() {constant=<rgba 1 1 0.4 0>}
  n48 = Transform(None=Normalize(VectorInput=Multiply(A=@n46, B=@n47))) {transformSourceType=TRANSFORMSOURCE_World, transformType=TRANSFORM_View}
  n49 = Multiply(A=Normalize(VectorInput=ComponentMask(None=@n48) {g=true, r=true}), B=Constant2Vector() {g=-1, r=1})
  n50 = WorldPosition()
  n51 = Add(A=@n50.XYZ, B=Multiply(A=@n50.Z) {constB=0.44})
  n52 = Multiply(A=ComponentMask(None=@n50.XYZ) {r=true}) {constB=0.2}
  n53 = Frac(None=Divide(A=Add(A=@n51, B=@n52)) {constB=60})
  n54 = Abs(None=Multiply(A=Add(A=@n53) {constB=-0.5}) {constB=2})
  n55 = OneMinus(None=Multiply(A=@n54, B=@n54) {constB=0.1})
  n56 = Multiply(A=@n55, B=@n55)
  n57 = Saturate(None=Divide(A=NamedRerouteUsage()) {constB=2000})
  n58 = Multiply(A=DotProduct(A=OneMinus(None=@n56), B=Constant() {r=1}), B=LinearInterpolate(Alpha=SquareRoot(None=@n57)) {constA=1.5, constB=0.2})
  n59 = Multiply(A=Multiply(A=SceneTexelSize(), B=Add(A=@n58))) {constB=0.75}
  n60 = Multiply(A=@n49, B=Multiply(A=@n59) {constB=3})
  n61 = ComponentMask(None=SceneTexture(UVs=Add(A=@n45.ViewportUV, B=@n60)) {sceneTextureId=PPI_SceneDepth}.Color) {r=true}
  n62 = NamedRerouteUsage()
  n63 = Multiply(A=Step(Y=@n44, X=Subtract(A=@n61, B=@n62))) {constB=0.75}
  n64 = Multiply(A=@n59) {constB=2}
  n65 = SceneTexture(UVs=Add(A=@n45.ViewportUV, B=Multiply(A=@n49, B=@n64))) {sceneTextureId=PPI_SceneDepth}
  n66 = Step(Y=@n44, X=Subtract(A=ComponentMask(None=@n65.Color) {r=true}, B=@n62))
  n67 = Add(A=@n45.ViewportUV, B=Multiply(A=@n49, B=Multiply(A=@n59)))
  n68 = Subtract(A=ComponentMask(None=SceneTexture(UVs=@n67) {sceneTextureId=PPI_SceneDepth}.Color) {r=true}, B=@n62)
  NamedRerouteDeclaration(None=Add(A=Add(A=@n63, B=@n66), B=Step(Y=@n44, X=@n68))) {name=Highlight Light, nodeColor=<rgba 0 0.2734 1 1>}
  n69 = Subtract(A=NamedRerouteUsage()) {constB=5000}
  n70 = CollectionParameter() {collection=&MPC_Ch8_PP, parameterName=GammaDistanceFalloff}
  NamedRerouteDeclaration(None=Saturate(None=Divide(A=@n69, B=@n70) {constB=2000})) {name=Color Correction Fade, nodeColor=<rgba 1 0 0.2437 1>}
  n71 = Subtract(A=NamedRerouteUsage(), B=ScalarParameter() {defaultValue=300000, group=Outlines Opacity, parameterName=Sky Cutoff}) {constB=20000}
  n72 = OneMinus(None=Saturate(None=Divide(A=@n71) {constB=300000}))
  NamedRerouteDeclaration(None=Multiply(A=@n72, B=@n72)) {name=Sky Cutoff, nodeColor=<rgba 0.8047 0 1 1>}
  n73 = NamedRerouteUsage()
  n74 = Subtract(A=@n73, B=ScalarParameter() {defaultValue=5000, group=Outlines Opacity, parameterName=Opacity Fade Start}) {constB=20000}
  n75 = CollectionParameter() {collection=&MPC_Ch8_PP, parameterName=OutlineDistanceFalloff}
  NamedRerouteDeclaration(None=Saturate(None=Divide(A=@n74, B=@n75) {constB=2000})) {name=Distance Fade, nodeColor=<rgba 1 0.4219 0 1>}
  n76 = SceneTexture(UVs=NamedRerouteUsage()) {sceneTextureId=PPI_WorldNormal}
  n77 = SceneTexture(UVs=NamedRerouteUsage()) {sceneTextureId=PPI_WorldNormal}
  n78 = DotProduct(
   A=NamedRerouteDeclaration(None=ComponentMask(None=@n76.Color) {b=true, g=true, r=true}) {name=Normal 1, nodeColor=<rgba 0.4568 0 1 1>}
   B=NamedRerouteDeclaration(None=ComponentMask(None=@n77.Color) {b=true, g=true, r=true}) {name=Normal 2, nodeColor=<rgba 0.4568 0 1 1>}
  )
  n79 = SceneTexture(UVs=NamedRerouteUsage()) {sceneTextureId=PPI_WorldNormal}
  n80 = SceneTexture(UVs=NamedRerouteUsage()) {sceneTextureId=PPI_WorldNormal}
  n81 = DotProduct(
   A=NamedRerouteDeclaration(None=ComponentMask(None=@n79.Color) {b=true, g=true, r=true}) {name=Normal 3, nodeColor=<rgba 0.4568 0 1 1>}
   B=NamedRerouteDeclaration(None=ComponentMask(None=@n80.Color) {b=true, g=true, r=true}) {name=Normal 4, nodeColor=<rgba 0.4568 0 1 1>}
  )
  n82 = Step(
   Y=ScalarParameter() {defaultValue=0.9, group=Outlines Normal, parameterName=NormalThreshold, sliderMax=30, sortPriority=1}
   X=Min(A=Saturate(None=@n78), B=Saturate(None=@n81))
  )
  n83 = NamedRerouteUsage()
  NamedRerouteDeclaration(None=Max(A=Max(A=@n82, B=@n83), B=NamedRerouteUsage())) {name=Normal Outlines, nodeColor=<rgba 0.5703 1 0 1>}
  n84 = ScalarParameter() {defaultValue=30000, group=Outlines Opacity, parameterName=Distant Cutoff}
  NamedRerouteDeclaration(None=Saturate(None=Subtract(A=@n73, B=@n84) {constB=20000})) {name=Distance Cutoff, nodeColor=<rgba 0.8047 0 1 1>}
  n85 = NamedRerouteUsage()
  n86 = NamedRerouteUsage()
  n87 = Min(
   A=ComponentMask(None=SceneTexture(UVs=@n85) {sceneTextureId=PPI_CustomDepth}.Color) {r=true}
   B=ComponentMask(None=SceneTexture(UVs=@n86) {sceneTextureId=PPI_CustomDepth}.Color) {r=true}
  )
  n88 = NamedRerouteUsage()
  n89 = NamedRerouteUsage()
  n90 = Min(
   A=ComponentMask(None=SceneTexture(UVs=@n88) {sceneTextureId=PPI_CustomDepth}.Color) {r=true}
   B=ComponentMask(None=SceneTexture(UVs=@n89) {sceneTextureId=PPI_CustomDepth}.Color) {r=true}
  )
  n91 = Multiply(A=Subtract(A=Min(A=@n87, B=@n90), B=NamedRerouteUsage()), B=NamedRerouteUsage())
  NamedRerouteDeclaration(None=OneMinus(None=Saturate(None=@n91))) {name=No Outlines Mask, nodeColor=<rgba 0 0.25 1 1>}
  n92 = NamedRerouteUsage()
  n93 = NamedRerouteDeclaration(None=ComponentMask(None=SceneTexture(UVs=@n92) {sceneTextureId=PPI_SceneDepth}.Color) {r=true}) {name=Depth 1, nodeColor=<rgba 1 1 1 1>}
  n94 = NamedRerouteUsage()
  n95 = NamedRerouteDeclaration(None=ComponentMask(None=SceneTexture(UVs=@n94) {sceneTextureId=PPI_SceneDepth}.Color) {r=true}) {name=Depth 2, nodeColor=<rgba 1 1 1 1>}
  n96 = NamedRerouteUsage()
  n97 = NamedRerouteDeclaration(None=ComponentMask(None=SceneTexture(UVs=@n96) {sceneTextureId=PPI_SceneDepth}.Color) {r=true}) {name=Depth 3, nodeColor=<rgba 1 1 1 1>}
  n98 = NamedRerouteUsage()
  n99 = NamedRerouteDeclaration(None=ComponentMask(None=SceneTexture(UVs=@n98) {sceneTextureId=PPI_SceneDepth}.Color) {r=true}) {name=Depth 4, nodeColor=<rgba 1 1 1 1>}
  n100 = Max(A=Abs(None=Subtract(A=@n93, B=@n95)), B=Abs(None=Subtract(A=@n97, B=@n99)))
  n101 = Min(A=Min(A=@n93, B=@n95), B=Min(A=@n97, B=@n99))
  n102 = SceneTexelSize()
  n103 = Multiply(
   A=Multiply(
    A=ScalarParameter() {defaultValue=80, group=Outlines Depth, parameterName=DepthThreshold, sliderMax=30, sortPriority=1}
    B=ComponentMask(None=@n102) {g=true}
   )
   B=Add(A=NamedRerouteUsage())
  )
  n104 = Multiply(A=NamedRerouteUsage()) {constB=0.1}
  n105 = NamedRerouteDeclaration(None=Multiply(A=NamedRerouteDeclaration(None=@n101) {name=Min Depth, nodeColor=<rgba 1 0 0.7734 1>}, B=Add(A=@n103, B=@n104))) {name=Depth Threshold, nodeColor=<rgba 1 0.3516 0 1>}
  NamedRerouteDeclaration(None=Max(A=Step(Y=@n100, X=@n105), B=NamedRerouteUsage())) {name=Depth Outlines, nodeColor=<rgba 1 0.25 1 1>}
  n106 = Multiply(A=@n49, B=Multiply(A=SceneTexelSize()) {constB=-1})
  n107 = ComponentMask(None=SceneTexture(UVs=Add(A=@n45.ViewportUV, B=@n106)) {sceneTextureId=PPI_SceneDepth}.Color) {r=true}
  NamedRerouteDeclaration(None=Step(Y=Subtract(A=@n61, B=@n107), X=@n44)) {name=Highlights Mask, nodeColor=<rgba 1 0.25 1 1>}
  NamedRerouteDeclaration(None=Max(A=@n17, B=NamedRerouteUsage())) {name=FinalOutline, nodeColor=<rgba 0 1 0.0625 1>}
  n108 = ComponentMask(None=SceneTexture() {sceneTextureId=PPI_CustomDepth}.Color) {r=true}
  n109 = NamedRerouteDeclaration(None=ComponentMask(None=SceneTexture() {sceneTextureId=PPI_SceneDepth}.Color) {r=true}) {name=Depth, nodeColor=<rgba 1 0.2344 0 1>}
  NamedRerouteDeclaration(None=Saturate(None=Subtract(A=@n108, B=@n109))) {name=No Color Correction Mask, nodeColor=<rgba 0.2891 1 0 1>}
  n110 = Max(A=Max(A=@n21, B=@n22), B=Max(A=@n23, B=@n24))
  n111 = Abs(None=Subtract(A=@n110) {constB=10})
  NamedRerouteDeclaration(None=OneMinus(None=Saturate(None=@n111))) {name=Water Mask, nodeColor=<rgba 0 0.9062 1 1>}
  n112 = Abs(None=Subtract(A=@n110) {constB=2})
  NamedRerouteDeclaration(None=OneMinus(None=Saturate(None=@n112))) {name=Vegetation Mask, nodeColor=<rgba 0 0.9062 1 1>}
  n113 = ComponentMask(None=SceneTexture(UVs=NamedRerouteUsage()) {sceneTextureId=PPI_PostProcessInput0}.Color) {b=true, g=true, r=true}
  n114 = ComponentMask(None=SceneTexture(UVs=NamedRerouteUsage()) {sceneTextureId=PPI_PostProcessInput0}.Color) {b=true, g=true, r=true}
  n115 = ComponentMask(None=SceneTexture(UVs=NamedRerouteUsage()) {sceneTextureId=PPI_PostProcessInput0}.Color) {b=true, g=true, r=true}
  n116 = ComponentMask(None=SceneTexture(UVs=NamedRerouteUsage()) {sceneTextureId=PPI_PostProcessInput0}.Color) {b=true, g=true, r=true}
  NamedRerouteDeclaration(None=Min(A=Min(A=@n113, B=@n114), B=Min(A=@n115, B=@n116))) {name=Posterized Min Color, nodeColor=<rgba 0.3125 0 1 1>}
  n117 = ComponentMask(None=ScreenPosition().ViewportUV) {g=true, r=true}
  n118 = ComponentMask(None=ScreenPosition().ViewportUV) {g=true, r=true}
  n119 = Constant()
  n120 = Multiply(A=SceneTexelSize(), B=ScalarParameter() {defaultValue=1, group=LineThicknessOveride for Capture, parameterName=LineThicknessOveride})
  n121 = ComponentMask(None=@n120) {g=true}
  n122 = AppendVector(A=@n119, B=@n121)
  NamedRerouteDeclaration(None=LinearInterpolate(A=@n117, B=Add(A=@n118, B=@n122), Alpha=NamedRerouteUsage())) {name=UV1, nodeColor=<rgba 1 0 0.3763 1>}
  n123 = Subtract(A=NamedRerouteDeclaration(None=@n113) {name=First Color Sample, nodeColor=<rgba 0 0.4141 1 1>}, B=@n114)
  n124 = Subtract(A=@n115, B=@n116)
  MaterialExpressionSaturate_17 = Saturate(None=Max(A=Abs(None=@n123), B=Abs(None=@n124)))
  n125 = Convert(None=@MaterialExpressionSaturate_17) {
   nodeName=Break Float3
   convertInputs [{
     type=Vector3
     expressionInput {
      expression=&self:MaterialExpressionSaturate_17
      inputName=None
      mask=0
      maskA=0
      maskB=0
      maskG=0
      maskR=0
      outputIndex=0
     }
    }]
   convertOutputs [{type=Scalar}, {type=Scalar}, {type=Scalar}]
  }
  n126 = Max(A=Max(A=@n125, B=@n125), B=@n125)
  n127 = OneMinus(None=@n126)
  n128 = ScalarParameter() {defaultValue=32, parameterName=Posterized Power}
  n129 = Saturate(None=Multiply(A=Power(Base=@n127, Exp=@n128) {constExponent=12}) {constB=7})
  NamedRerouteDeclaration(None=@n129) {name=Posterized Lines, nodeColor=<rgba 0.8281 1 0 1>}
  n130 = Add(A=Min(A=NamedRerouteUsage(), B=NamedRerouteUsage()), B=Add(A=NamedRerouteUsage(), B=NamedRerouteUsage()))
  n131 = LinearInterpolate(
   A=ScalarParameter() {defaultValue=0.53, parameterName=Posterized Far Opacity}
   B=ScalarParameter() {defaultValue=0.8, parameterName=Posterized Close Opacity}
   Alpha=NamedRerouteUsage()
  ) {constA=0.1, constB=0.6}
  n132 = Multiply(A=Saturate(None=Add(A=@n130, B=@n131)), B=NamedRerouteUsage())
  n133 = SceneTexture() {sceneTextureId=PPI_PostProcessInput0}
  NamedRerouteDeclaration(None=LinearInterpolate(A=@n132, B=ComponentMask(None=@n133.Color) {b=true, g=true, r=true}, Alpha=Saturate(None=@n130)) {constB=0}) {name=Mobile Posturizer V2, nodeColor=<rgba 1 0.0293 0.6616 1>}
  n134 = Abs(None=Subtract(A=NamedRerouteUsage(), B=NamedRerouteUsage()))
  n135 = Abs(None=Subtract(A=NamedRerouteUsage(), B=NamedRerouteUsage()))
  n136 = NamedRerouteUsage()
  n137 = SceneTexelSize()
  n138 = Multiply(
   A=Multiply(
    A=ScalarParameter() {defaultValue=200, group=Outlines Depth, parameterName=MobileDepthThreshold, sliderMax=30, sortPriority=1}
    B=ComponentMask(None=@n137) {g=true}
   )
   B=Add(A=NamedRerouteUsage())
  )
  NamedRerouteDeclaration(None=Step(Y=Max(A=@n134, B=@n135), X=Multiply(A=@n136, B=@n138))) {name=Mobile Depth Lines, nodeColor=<rgba 0.0324 1 0 1>}
  n139 = ComponentMask(None=SceneTexture() {sceneTextureId=PPI_PostProcessInput0}.Color) {b=true, g=true, r=true}
  n140 = DotProduct(A=SkyAtmosphereLightDirection(), B=NamedRerouteUsage())
  n141 = Saturate(None=Add(A=Multiply(A=@n140) {constB=20}) {constB=0.5})
  n142 = OneMinus(None=NamedRerouteUsage())
  n143 = NamedRerouteDeclaration(None=LinearInterpolate(A=Multiply(A=@n139) {constB=0.6}, B=@n139, Alpha=Max(A=@n141, B=@n142)) {constA=0.5, constB=0}) {name=Final Color, nodeColor=<rgba 0 1 0.2734 1>}
  n144 = LinearInterpolate(
   A=@n143
   B=Multiply(
    A=SquareRoot(None=@n143)
    B=CollectionParameter() {collection=&MPC_Ch8_PP, parameterName=GammaMultiply}
   ) {constB=0.35}
   Alpha=LinearInterpolate(
    A=CollectionParameter() {collection=&MPC_Ch8_PP, parameterName=GammaBlendNear}
    B=CollectionParameter() {collection=&MPC_Ch8_PP, parameterName=GammaBlendFar}
    Alpha=NamedRerouteUsage()
   ) {constA=0.9, constB=0.2}
  ) {constAlpha=0.3}
  n145 = Multiply(A=LinearInterpolate(
    A=CollectionParameter() {collection=&MPC_Ch8_PP, parameterName=GammaSaturationNear}
    B=CollectionParameter() {collection=&MPC_Ch8_PP, parameterName=GammaSaturationFar}
    Alpha=NamedRerouteUsage()
   ) {constA=1, constB=0.5}) {constB=-1}
  NamedRerouteDeclaration(None=EyeAdaptationInverse(
    LightValueInput=Desaturation(None=@n144, Fraction=@n145)
    AlphaInput=CollectionParameter() {collection=&MPC_Ch8_PP, parameterName=GammaInverseEyeAdaptation}
   ).EyeAdaptationInverse) {name=Final Color Treated, nodeColor=<rgba 0.2656 1 0 1>}
  n146 = NamedRerouteUsage()
  n147 = CollectionParameter() {collection=&MPC_Ch8_PP, parameterName=HighlightDistanceFalloff}
  NamedRerouteDeclaration(None=Saturate(None=Divide(A=@n146, B=@n147) {constB=2000})) {name=Highlight Fade, nodeColor=<rgba 1 0 0.2437 1>}
  n148 = SceneTexelSize()
  n149 = ComponentMask(None=@n148) {r=true}
  n150 = ComponentMask(None=@n148) {g=true}
  n151 = AppendVector(A=@n149, B=@n150)
  n152 = AppendVector(A=@n119, B=Multiply(A=@n121) {constB=-1})
  NamedRerouteDeclaration(None=LinearInterpolate(A=Subtract(A=@n117, B=@n151), B=Add(A=@n118, B=@n152) {desc=y+}, Alpha=NamedRerouteUsage())) {name=UV2, nodeColor=<rgba 0 0.6514 1 1>}
  n153 = Constant()
  n154 = AppendVector(A=@n153, B=@n150)
  n155 = ComponentMask(None=@n120) {r=true}
  n156 = AppendVector(A=@n155, B=@n119)
  NamedRerouteDeclaration(None=LinearInterpolate(A=Subtract(A=@n117, B=@n154), B=Add(A=@n118, B=@n156) {desc=x-}, Alpha=NamedRerouteUsage())) {name=UV3, nodeColor=<rgba 0.5 0 1 1>}
  n157 = AppendVector(A=@n149, B=@n153)
  n158 = AppendVector(A=Multiply(A=@n155) {constB=-1}, B=@n119)
  NamedRerouteDeclaration(None=LinearInterpolate(A=Subtract(A=@n117, B=@n157), B=Add(A=@n118, B=@n158) {desc=y-}, Alpha=NamedRerouteUsage())) {name=UV4, nodeColor=<rgba 0.6641 1 0 1>}
  n159 = CameraVectorWS()
  n160 = Max(A=DotProduct(A=@n159, B=NamedRerouteUsage()), B=DotProduct(A=@n159, B=NamedRerouteUsage()))
  n161 = Max(A=DotProduct(A=@n159, B=NamedRerouteUsage()), B=DotProduct(A=@n159, B=NamedRerouteUsage()))
  NamedRerouteDeclaration(None=Step(
    Y=Max(A=@n160, B=@n161)
    X=ScalarParameter() {defaultValue=0.1, group=Outlines Depth, parameterName=DepthFresnel, sliderMax=30, sortPriority=1}
   ) {constX=0.05}) {name=Fresnel, nodeColor=<rgba 1 0 0.1172 1>}
  NamedRerouteUsage()
  OneMinus(None=@n129)
  OneMinus(None=@n132)
  Step(Y=@n126, X=ScalarParameter() {defaultValue=0.1, parameterName=Posterized Power 2})
 ]
}
