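# Post-process material, applied after DOF, that adds a brightened, saturation-boosted rim highlight to depth edges facing the sky atmosphere light.
# It projects the vertically flattened light direction into screen space and samples scene depth at three stepped offsets along it.
# Depth differences above a distance-scaled threshold sum into a weighted mask, which fades with distance and blends the highlight over scene color.
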Material /Game/TimeOfDay/DaySequence/DaySequenceActors/BR_Ch8/PPM_Highlights
graph {
 properties {
  bAutomaticallySetUsageInEditor=false
  blendableLocation=BL_SceneColorAfterDOF
  materialDomain=MD_PostProcess
  previewMesh=/Hera/Sets/SeaMist/Meshes/SM_SeaMist_Lighthouse_FresnelLens_A
  usedShadingModels=Default Lit
 }
 nodes [
  n1 = ComponentMask(None=SceneTexture() {sceneTextureId=PPI_PostProcessInput0}.Color) {b=true, g=true, r=true}
  n2 = Multiply(A=ScalarParameter() {defaultValue=0.5, parameterName=Edge Highlight Saturation}) {constB=-1}
  n3 = Saturate(None=Divide(A=NamedRerouteUsage(), B=ScalarParameter() {defaultValue=30000, group=Outlines Opacity, parameterName=Edge Highlight Fade Distance}) {constB=2000})
  LinearInterpolate(
   A=@n1
   B=Multiply(A=Desaturation(None=@n1, Fraction=@n2), B=ScalarParameter() {defaultValue=2, parameterName=Edge Highlight Intensity}) {constB=2}
   Alpha=Multiply(A=NamedRerouteUsage(), B=OneMinus(None=@n3))
  )
  n4 = SceneTexelSize()
  n5 = Multiply(
   A=NamedRerouteUsage()
   B=Multiply(A=ComponentMask(None=@n4) {g=true}, B=ScalarParameter() {defaultValue=80, parameterName=Edge Highlight Threshold}) {constB=80}
  )
  n6 = ScreenPosition()
  n7 = SkyAtmosphereLightDirection()
  n8 = Constant3Vector() {constant=<rgba 1 1 0.4 0>}
  n9 = Transform(None=Normalize(VectorInput=Multiply(A=@n7, B=@n8))) {transformSourceType=TRANSFORMSOURCE_World, transformType=TRANSFORM_View}
  n10 = Multiply(A=Normalize(VectorInput=ComponentMask(None=@n9) {g=true, r=true}), B=Constant2Vector() {g=-1, r=1})
  n11 = Multiply(A=SceneTexelSize(), B=ScalarParameter() {defaultValue=1, group=Outlines Opacity, parameterName=Edge Highlights Width}) {constB=2}
  n12 = Multiply(A=@n10, B=Multiply(A=@n11) {constB=3})
  n13 = ComponentMask(None=SceneTexture(UVs=Add(A=@n6.ViewportUV, B=@n12)) {sceneTextureId=PPI_SceneDepth}.Color) {r=true}
  n14 = NamedRerouteUsage()
  n15 = Multiply(A=Step(Y=@n5, X=Subtract(A=@n13, B=@n14))) {constB=0.75}
  n16 = Multiply(A=@n11) {constB=2}
  n17 = SceneTexture(UVs=Add(A=@n6.ViewportUV, B=Multiply(A=@n10, B=@n16))) {sceneTextureId=PPI_SceneDepth}
  n18 = Step(Y=@n5, X=Subtract(A=ComponentMask(None=@n17.Color) {r=true}, B=@n14))
  n19 = Add(A=@n6.ViewportUV, B=Multiply(A=@n10, B=Multiply(A=@n11)))
  n20 = Subtract(A=ComponentMask(None=SceneTexture(UVs=@n19) {sceneTextureId=PPI_SceneDepth}.Color) {r=true}, B=@n14)
  NamedRerouteDeclaration(None=Add(A=Add(A=@n15, B=@n18), B=Step(Y=@n5, X=@n20))) {name=Highlight Light, nodeColor=<rgba 0 0.2734 1 1>}
  NamedRerouteDeclaration(None=ComponentMask(None=SceneTexture() {sceneTextureId=PPI_SceneDepth}.Color) {r=true}) {name=Depth, nodeColor=<rgba 1 0.2344 0 1>}
 ]
}
